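# Sound cue for Sparks characters that plays a guitar "ready" sound, judging by its name and wave files.
# A root Random node picks among seven GtrReady wave variants with equal weight, so repeated triggers vary.
# The cue routes to the in-game SFX sound class and sets no attenuation or pitch and volume modulation of its own.
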SoundCue /SparksCharacterCommon/Sounds/Audio_FM_GtrReady
root=Random_0
nodes [
 {
  id=Random_0
  children [WavePlayer_0,WavePlayer_1,WavePlayer_2,WavePlayer_3,WavePlayer_4,WavePlayer_5,WavePlayer_6]
  properties {weights [1,1,1,1,1,1,1]}
 }
 {id=WavePlayer_0, properties {soundWaveAssetPtr=/SparksCharacterCommon/Sounds/wavs/GtrReady01}}
 {id=WavePlayer_1, properties {soundWaveAssetPtr=/SparksCharacterCommon/Sounds/wavs/GtrReady02}}
 {id=WavePlayer_2, properties {soundWaveAssetPtr=/SparksCharacterCommon/Sounds/wavs/GtrReady03}}
 {id=WavePlayer_3, properties {soundWaveAssetPtr=/SparksCharacterCommon/Sounds/wavs/GtrReady04}}
 {id=WavePlayer_4, properties {soundWaveAssetPtr=/SparksCharacterCommon/Sounds/wavs/GtrReady05}}
 {id=WavePlayer_5, properties {soundWaveAssetPtr=/SparksCharacterCommon/Sounds/wavs/GtrReady06}}
 {id=WavePlayer_6, properties {soundWaveAssetPtr=/SparksCharacterCommon/Sounds/wavs/GtrReady07}}
]
properties {maxDistance=2097152, soundClassObject=&Fort_SFX_InGame, volumeMultiplier=1}
